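# Sound cue for the DadBro boss's Crawling Bolt summon, playing a looping lightning sound borrowed from Storm King assets.
# An equally weighted Random node picks one of three looping wave variants and feeds it through a Modulator with fixed volume.
# The cue routes to the boss fight SFX class, overrides attenuation, and caps concurrency at a few instances, blocking new ones beyond that.
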SoundCue /FNGameplayCues/Sounds/Apollo/Events/Dadbro/DadBro_CrawlingBolt_Summon_Cue
root=Modulator_3
nodes [
 {id=Modulator_3, children [Random_2], properties {volumeMax=1, volumeMin=1}}
 {id=Random_2, children [WavePlayer_2,WavePlayer_5,WavePlayer_7], properties {weights [1,1,1]}}
 {
  id=WavePlayer_2
  properties {bLooping=true, soundWaveAssetPtr=/Game/Sounds/Fort_Monster_Storm_King/SFX/Final_SFX/StormKing_LightningSummons_Lightning_Prct_Loop_01}
 }
 {
  id=WavePlayer_5
  properties {bLooping=true, soundWaveAssetPtr=/Game/Sounds/Fort_Monster_Storm_King/SFX/Final_SFX/StormKing_LightningSummons_Lightning_Prct_Loop_02}
 }
 {
  id=WavePlayer_7
  properties {bLooping=true, soundWaveAssetPtr=/Game/Sounds/Fort_Monster_Storm_King/SFX/Final_SFX/StormKing_LightningSummons_Lightning_Prct_Loop_03}
 }
]
properties {
 bOverrideAttenuation=true
 bOverrideConcurrency=true
 maxDistance=6000
 soundClassObject=&Fort_BossFight_SFX
 volumeMultiplier=1
 attenuationOverrides {falloffDistance=4000, attenuationShapeExtents {x=2000}}
 concurrencyOverrides {maxCount=3, resolutionRule=PreventNew}
}
