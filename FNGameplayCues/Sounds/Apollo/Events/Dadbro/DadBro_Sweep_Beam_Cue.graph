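# Sound cue that plays the looping sweep beam sound for the Dadbro boss event, reusing a Storm King beam attack loop.
# A short random delay feeds an Enveloper whose volume curve fades the looping wave in over roughly a quarter second.
# It uses a dedicated attenuation asset, the boss fight player-critical sound class, boosted volume, and a concurrency override that prevents new instances at the limit.
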SoundCue /FNGameplayCues/Sounds/Apollo/Events/Dadbro/DadBro_Sweep_Beam_Cue
root=Delay_0
nodes [
 {id=Delay_0, children [Enveloper_0], properties {delayMax=0.1, delayMin=0.05}}
 {
  id=Enveloper_0
  children [WavePlayer_2]
  properties {pitchCurve {editorCurveData {keys [{time=0}]}}, volumeCurve {editorCurveData {keys [<at -0.0028 0>,<at 0.2511 1.1396>]}}}
 }
 {
  id=WavePlayer_2
  properties {bLooping=true, soundWaveAssetPtr=/FNGameplayCues/Sounds/Fort_Monster_Storm_King/SFX/StormKing_BeamAttack_Beam_01_Loop}
 }
]
properties {
 attenuationSettings=/FNGameplayCues/Sounds/Apollo/Events/Dadbro/Attenuation/Dadbro_Sweep_Beam_Att
 bOverrideConcurrency=true
 maxDistance=6000
 soundClassObject=/Game/Sounds/FortSoundClassesAndModes/Fort_BossFight_SoundClass/Fort_BossFight_PlayerCritical
 volumeMultiplier=1.5
 attenuationOverrides {attenuationShape=Capsule, distanceAlgorithm=Logarithmic, falloffDistance=3000, attenuationShapeExtents=<3000 512>}
 concurrencyOverrides {resolutionRule=PreventNew}
}
