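# Sound cue for the DadBro vent launch event, playing one of four DudeBro vent launch recordings.
# An equally weighted random node feeds a modulator, then a Doppler node at the root.
# It uses ability status-effect attenuation and the ambient sound class at reduced volume.
# A concurrency override allows one instance per owner and blocks new plays while it sounds.
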SoundCue /FNGameplayCues/Sounds/Apollo/Events/Dadbro/DadBro_Vent_Launch_Cue
root=Doppler_0
nodes [
 {id=Doppler_0, children [Modulator_0]}
 {id=Modulator_0, children [Random_0]}
 {id=Random_0, children [WavePlayer_2,WavePlayer_3,WavePlayer_4,WavePlayer_5], properties {weights [1,1,1,1]}}
 {id=WavePlayer_2, properties {soundWaveAssetPtr=/FNGameplayCues/Athena/Sounds/Events/CubeIsland/DudeBro_Vent_Launch_04}}
 {id=WavePlayer_3, properties {soundWaveAssetPtr=/FNGameplayCues/Athena/Sounds/Events/CubeIsland/DudeBro_Vent_Launch_01}}
 {id=WavePlayer_4, properties {soundWaveAssetPtr=/FNGameplayCues/Athena/Sounds/Events/CubeIsland/DudeBro_Vent_Launch_02}}
 {id=WavePlayer_5, properties {soundWaveAssetPtr=/FNGameplayCues/Athena/Sounds/Events/CubeIsland/DudeBro_Vent_Launch_03}}
]
properties {
 attenuationSettings=/Game/Sounds/Attenuation/Fort_Ability_Status_Effect_Activation_Att
 bOverrideConcurrency=true
 maxDistance=6000
 soundClassObject=&Fort_Ambient_Athena
 volumeMultiplier=0.176
 concurrencyOverrides {bLimitToOwner=true, maxCount=1, resolutionRule=PreventNew}
}
